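; Gameplay ability that credits quest progress when the owner eliminates a downed (DBNO) opponent with a harvesting tool equipped.
; On event activation, it checks the payload's target tags for the DBNO tag, then the avatar's ability system for the equipped melee harvest tag.
; If both match, it sends a complex custom stat event with configurable tags to the owning player's quest manager.
; It ends after each check, with debug-gated prints tracing the outcome.
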(event Ability|EventActivateAbilityFromEvent
  (params (EventData :type Gameplay%20Event%20Data%20Structure%20%28by%20ref%29))
  (bind (_eventtag _instigator _target _optionalobject _optionalobject2 _contexthandle _instigatortags _targettags) (Utilities|Struct|BreakGameplayEventData EventData))
  (|CallParentFunction EventData)
  (Development|PrintString "@@@ ActivateAbilityFromEvent [GA_Quest_Harvest_Downed_Opponent]: CALLED!" false (Variables|Debug|GetbDebugAbility))
  (if (GameplayTags|HasTag _targettags (Variables|Default|GetTargetDBNOTag))
    (Development|PrintString "@@@@@@ ActivateAbilityFromEvent [GA_Quest_Harvest_Downed_Opponent]: Killed target is DBNO!" false (Variables|Debug|GetbDebugAbility))
    (if (GameplayTags|HasAnyMatchingGameplayTags (Ability|GetAbilitySystemComponent (Ability|GetAvatarActorfromActorInfo)) (GameplayTags|MakeLiteralGameplayTagContainer "(GameplayTags=((TagName=\"Granted.ItemEquipped.Melee.Harvest\")))"))
      (Development|PrintString "@@@@@@@@@ ActivateAbilityFromEvent [GA_Quest_Harvest_Downed_Opponent]: Source HAS Harvester equipped!!" false (Variables|Debug|GetbDebugAbility))
      (Quest|SendComplexCustomStatEvent (Variables|GAVars|Data|GetOwningPlayerQuestManager) 0 (Variables|Default|GetQuestStatEventTags) "(GameplayTags=)" 1)
      (else
        (Development|PrintString "@@@@@@@@@ ActivateAbilityFromEvent [GA_Quest_Harvest_Downed_Opponent]: Source DOES NOT have Harvester equipped!!" false (Variables|Debug|GetbDebugAbility))))
    (Ability|EndAbility)
    (else
      (Development|PrintString "@@@@@@ ActivateAbilityFromEvent [GA_Quest_Harvest_Downed_Opponent]: Killed target is NOT DBNO!" false (Variables|Debug|GetbDebugAbility))
      (Ability|EndAbility))))
